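# Niagara system for the PuffinSmile pickaxe's swing trail, drawn as short-lived ribbons.
# One emitter, MainTrails, uses FN anim-trail modules to spawn and update particles from the system's skeletal mesh, feeding two ribbon renderers.
# Color blends grey with a time-of-day color and scales by a time-of-day float.
# A cubic curve scaled by SwingGeneralFade drives both trail alpha and the alpha of a separate black LineColor.
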NiagaraSystem /BRCosmetics/Weapons/FORT_Melee/Pickaxe_PuffinSmile/FX/NS_Pickaxe_PuffinSmile_Trail
dependencies {
 usedDataInterfaces [&NiagaraDataInterfaceCurve, &NiagaraDataInterfaceSkeletalMesh, &NiagaraDataInterfaceColorCurve]
 usedDynamicInputs [&ColorFromCurve, &FNTimeOfDayScalar]
 usedModules [
  &self:SystemSpawnScript.SystemScriptSource.SystemScriptGraph.NiagaraNodeAssignment_0.SetVariables_#1
  &SystemState
  &self:SystemSpawnScript.SystemScriptSource.SystemScriptGraph.NiagaraNodeAssignment_1.SetVariables_#2
  &self:SystemSpawnScript.SystemScriptSource.SystemScriptGraph.NiagaraNodeAssignment_2.SetVariables_#3
 ]
}
emitters {MainTrails {
  emitterSpawnScript [&FN_RibbonIDOffsetUtility()]
  emitterUpdateScript [
   &EmitterState()
   &FN_SpawnAnimTrail()
  ]
  particleSpawnScript [
   &FN_InitializeAnimTrailBoneIndecies()
   &FN_InitializeAnimTrail(Color=<rgba 1 1 1 0.4>, Lifetime=0.13, SkeletalMesh=$System.Skeletal Mesh)
  ]
  particleUpdateScript [
   &FN_UpdateAnimTrail(SkeletalMesh=$System.Skeletal Mesh)
   &ScaleColor(
    Scale Alpha=&Multiply_Float(
     A=&FloatFromCurve(
      FloatCurve=DataInterface() {propertyValues {Curve {keys [{interpMode=RCIM_Cubic, leaveTangent=-4.0138, tangentMode=RCTM_Break, time=0, value=1}, <at 0.4 0>]}}}
      Scale Curve=$System.SwingGeneralFade
     )
     B=1
    )
    Scale RGB=&MakeVectorFromLinearColorRGB(Color=&LerpLinearColors(StartColor=<rgba 0.6927 0.6927 0.6927 1>, EndColor=$System.TODColor, LerpFactor=0.5))
    ScaleA=true
    ScaleRGB=true
   )
   &ScaleColor(Scale RGB=&VectorFromFloat(Value=$System.TODFloat), ScaleA=false, ScaleRGB=true) {module=ScaleColor001}
   /SetVariables(Particles.LineColor:LinearColor=&MakeLinearColorFromVectorAndFloat(
     "Vector (RGB)"=<0 0 0>
     "Float (Alpha)"=&Multiply_Float(
      A=&FloatFromCurve(
       FloatCurve=DataInterface() {propertyValues {Curve {keys [{interpMode=RCIM_Cubic, leaveTangent=-4.0138, tangentMode=RCTM_Break, time=0, value=1}, <at 0.4 0>]}}}
       Scale Curve=$System.SwingGeneralFade
      )
      B=1
     )
    )) {isSetParameters=true}
  ]
  rendererClasses [&NiagaraRibbonRendererProperties]
  renderers [
   {rendererClass=&NiagaraRibbonRendererProperties}
   {rendererClass=&NiagaraRibbonRendererProperties, rendererIndex=1}
  ]
 }}
summary {
 systemName=NS_Pickaxe_PuffinSmile_Trail
 emitters [{emitterName=MainTrails, rendererClasses [&NiagaraRibbonRendererProperties]}]
 userVariables [
  User.Debug {type=Vector3f, description="", defaultValue {struct=Vector3f, value=<0 0 0>}}
  User.RibbonIDOffset {type=NiagaraInt32}
 ]
}
