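# Physics asset for the PuffinSmile two-handed pickaxe cosmetic mesh, defining its simple collision shape.
# A single body on the Melee_TwoHanded_Handle bone uses one offset, rotated capsule.
# It has no constraints, so the mesh has only one physics body and no articulated parts.
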PhysicsAsset /BRCosmetics/Weapons/FORT_Melee/Pickaxe_PuffinSmile/Meshes/PuffinSmile_Axe_PhysicsAsset
bodies {Melee_TwoHanded_Handle [{shapeType=Capsule, length=20.7649, radius=5.7349, center=<-1.3296 -8.1477 41.9192>, rotation=<rot 0 -90 7.5762>}]}
constraints []
